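# Sound cue for the drop sound of the Happy Ghost weapon.
# A single Wave Player plays one drop wave with a slight volume boost, using the shared weapon-drop attenuation and the loot-collect sound class.
# It references the shared stop-oldest weapon-drop concurrency set and also stores single-instance, prevent-new override values, with no override flag shown as enabled.
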SoundCue /Game/Sounds/Fort_Weapons/Happy_Ghost/Happy_Ghost_Drop_Cue
root=WavePlayer_0
nodes [{id=WavePlayer_0, properties {soundWaveAssetPtr=/Game/Sounds/Fort_Weapons/Happy_Ghost/SoundWaves/Weapon_Happy_Ghost_Drop_01}}]
properties {
 attenuationSettings=&Fort_Weapon_Drop_Att
 maxDistance=3000
 soundClassObject=&Fort_Loot_Collect
 volumeMultiplier=1.2923
 concurrencyOverrides {maxCount=1, resolutionRule=PreventNew}
 concurrencySet [&Weapon_Drop_StopOldest]
}
