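# Sound cue for the Happy Ghost weapon's fire sound, layering close and medium-distance variations.
# A mixer plays two branches at once, each randomly picking one of three equally weighted waves.
# The close branch uses close projectile attenuation and the third-person fire sound class, while the medium branch uses medium mono attenuation and the distant fire class.
# Local and global gunfire concurrency sets govern simultaneous shots.
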SoundCue /Game/Sounds/Fort_Weapons/Happy_Ghost/Happy_Ghost_Fire_Cue
root=Mixer_1
nodes [
 {id=Attenuation_3, children [Random_0], properties {attenuationSettings=&Fort_Weapon_Projectile_Close_Athena_Att}}
 {
  id=Attenuation_4
  children [Random_2]
  properties {attenuationSettings=&Fort_Weapon_Projectile_Medium_Mono_Athena_Att}
 }
 {id=Mixer_1, children [SoundClass_3,SoundClass_2], properties {inputVolume [1,1]}}
 {id=Random_0, children [WavePlayer_2,WavePlayer_3,WavePlayer_4], properties {weights [1,1,1]}}
 {id=Random_2, children [WavePlayer_0,WavePlayer_1,WavePlayer_5], properties {weights [1,1,1]}}
 {
  id=SoundClass_2
  children [Attenuation_4]
  properties {soundClassOverride=&Fort_WeaponRanged_Fire_Distant}
 }
 {
  id=SoundClass_3
  children [Attenuation_3]
  properties {soundClassOverride=&Fort_WeaponRanged_Fire_3P_PSM}
 }
 {id=WavePlayer_0, properties {soundWaveAssetPtr=/Game/Sounds/Fort_Weapons/Happy_Ghost/SoundWaves/Weapon_Happy_Ghost_Fire_Med_01}}
 {id=WavePlayer_1, properties {soundWaveAssetPtr=/Game/Sounds/Fort_Weapons/Happy_Ghost/SoundWaves/Weapon_Happy_Ghost_Fire_Med_02}}
 {id=WavePlayer_2, properties {soundWaveAssetPtr=/Game/Sounds/Fort_Weapons/Happy_Ghost/SoundWaves/Weapon_Happy_Ghost_Fire_01}}
 {id=WavePlayer_3, properties {soundWaveAssetPtr=/Game/Sounds/Fort_Weapons/Happy_Ghost/SoundWaves/Weapon_Happy_Ghost_Fire_02}}
 {id=WavePlayer_4, properties {soundWaveAssetPtr=/Game/Sounds/Fort_Weapons/Happy_Ghost/SoundWaves/Weapon_Happy_Ghost_Fire_03}}
 {id=WavePlayer_5, properties {soundWaveAssetPtr=/Game/Sounds/Fort_Weapons/Happy_Ghost/SoundWaves/Weapon_Happy_Ghost_Fire_Med_03}}
]
properties {
 maxDistance=10579.2871
 volumeMultiplier=0.7
 attenuationOverrides {falloffDistance=5000, attenuationShapeExtents {x=1000}}
 concurrencyOverrides {resolutionRule=PreventNew}
 concurrencySet [&Weapon_Gunfire_Local_Conc, &Weapon_Gunfire_Global_Conc]
}
