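# Sound cue for the pickup sound of the Happy Ghost weapon.
# A single wave player feeds the output directly with one pickup wave, with no randomization or modulation.
# It uses the shared weapon pickup attenuation and the loot collect sound class, with a slight volume boost.
# A concurrency override allows only one instance at a time and blocks new plays while it sounds.
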SoundCue /Game/Sounds/Fort_Weapons/Happy_Ghost/Happy_Ghost_Pickup_Cue
root=WavePlayer_0
nodes [{id=WavePlayer_0, properties {soundWaveAssetPtr=/Game/Sounds/Fort_Weapons/Happy_Ghost/SoundWaves/Weapon_Happy_Ghost_Pickup_01}}]
properties {
 attenuationSettings=&Fort_Weapon_Pickup_Att
 bOverrideConcurrency=true
 maxDistance=3500
 soundClassObject=&Fort_Loot_Collect
 volumeMultiplier=1.3785
 concurrencyOverrides {maxCount=1, resolutionRule=PreventNew}
}
