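# Sound cue for the Happy Ghost weapon's pull-enemy whoosh, in the weapons sound class.
# A single wave player feeds the success whoosh sample through a modulator, with reduced overall volume.
# It references grappling gadget attenuation but enables cue-level overrides for falloff, non-spatialized radius and stereo spread, with reverb send disabled.
# Playback is limited by the player melee weapon concurrency set.
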SoundCue /Game/Sounds/Fort_Weapons/Happy_Ghost/Happy_Ghost_Pull_Enemy_Whoosh_Cue
root=Modulator_2
nodes [
 {id=Modulator_2, children [WavePlayer_4]}
 {id=WavePlayer_4, properties {soundWaveAssetPtr=/Game/Sounds/Fort_Weapons/Happy_Ghost/SoundWaves/Weapon_Happy_Ghost_Whoosh_Success_01}}
]
properties {
 attenuationSettings=/Game/Sounds/Attenuation/Fort_GadgetGrappling_Att
 bOverrideAttenuation=true
 maxDistance=3000
 soundClassObject=/Game/Sounds/FortSoundClassesAndModes/Fort_Weapons
 volumeMultiplier=0.3704
 attenuationOverrides {bEnableReverbSend=false, falloffDistance=3000, nonSpatializedRadiusStart=800, stereoSpread=2000, attenuationShapeExtents {x=0}}
 concurrencySet [&Player_Melee_Weapon_Conc]
}
