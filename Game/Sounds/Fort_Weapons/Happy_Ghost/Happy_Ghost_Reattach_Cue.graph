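# Sound cue for the reattach sound of the Happy Ghost weapon.
# A random node picks one of three equally weighted reattach wave variations on each play.
# It uses the grappling gadget attenuation and the weapons sound class at a reduced volume.
# A concurrency override prevents new instances from starting once the limit is reached.
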SoundCue /Game/Sounds/Fort_Weapons/Happy_Ghost/Happy_Ghost_Reattach_Cue
root=Random_0
nodes [
 {id=Random_0, children [WavePlayer_0,WavePlayer_1,WavePlayer_2], properties {weights [1,1,1]}}
 {id=WavePlayer_0, properties {soundWaveAssetPtr=/Game/Sounds/Fort_Weapons/Happy_Ghost/SoundWaves/Weapon_Happy_Ghost_ReAttach_01}}
 {id=WavePlayer_1, properties {soundWaveAssetPtr=/Game/Sounds/Fort_Weapons/Happy_Ghost/SoundWaves/Weapon_Happy_Ghost_ReAttach_02}}
 {id=WavePlayer_2, properties {soundWaveAssetPtr=/Game/Sounds/Fort_Weapons/Happy_Ghost/SoundWaves/Weapon_Happy_Ghost_ReAttach_03}}
]
properties {
 attenuationSettings=/Game/Sounds/Attenuation/Fort_GadgetGrappling_Att
 bOverrideConcurrency=true
 maxDistance=8500
 soundClassObject=/Game/Sounds/FortSoundClassesAndModes/Fort_Weapons
 volumeMultiplier=0.2593
 concurrencyOverrides {resolutionRule=PreventNew}
}
