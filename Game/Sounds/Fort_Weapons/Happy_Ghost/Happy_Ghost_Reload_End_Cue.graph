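# Sound cue for the end of the Happy Ghost weapon reload, which reuses the weapon's Equip_01 sound wave.
# A single wave player feeds the output directly, with no randomization or modulation.
# It plays at roughly half volume in the weapon foley sound class, using the shared weapon reload attenuation.
# The cue overrides concurrency and has its debug flag switched on.
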SoundCue /Game/Sounds/Fort_Weapons/Happy_Ghost/Happy_Ghost_Reload_End_Cue
root=WavePlayer_0
nodes [{id=WavePlayer_0, properties {soundWaveAssetPtr=/Game/Sounds/Fort_Weapons/Happy_Ghost/SoundWaves/Weapon_Happy_Ghost_Equip_01}}]
properties {
 attenuationSettings=&Fort_Weapon_Reload_Att
 bDebug=true
 bOverrideConcurrency=true
 maxDistance=2250
 soundClassObject=&Fort_WeaponFoley
 volumeMultiplier=0.501
}
