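# SoundCue for the Happy Ghost weapon's fish impact sound, classed under impacts.
# A mixer layers a randomly chosen medium water bullet-impact variation, with boosted input volume, over a fixed impact-success sound.
# It references gadget-activate attenuation settings, carries attenuation overrides that include disabled reverb send, and uses rifle bullet-impact concurrency.
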SoundCue /Game/Sounds/Fort_Weapons/Happy_Ghost/SC_Happy_Ghost_Impact_Fish
root=Mixer_1
nodes [
 {id=Mixer_1, children [Random_0,WavePlayer_5], properties {inputVolume [2.5,1]}}
 {id=Random_0, children [WavePlayer_8, WavePlayer_9, WavePlayer_10, WavePlayer_11, WavePlayer_12], properties {weights [1,1,1,1,1]}}
 {id=WavePlayer_10, properties {soundWaveAssetPtr=/Game/Sounds/Fort_Weapons/Happy_Ghost/SoundWaves/Bullet_Impact_Water_Med_03}}
 {id=WavePlayer_11, properties {soundWaveAssetPtr=/Game/Sounds/Fort_Weapons/Happy_Ghost/SoundWaves/Bullet_Impact_Water_Med_04}}
 {id=WavePlayer_12, properties {soundWaveAssetPtr=/Game/Sounds/Fort_Weapons/Happy_Ghost/SoundWaves/Bullet_Impact_Water_Med_05}}
 {id=WavePlayer_5, properties {soundWaveAssetPtr=/Game/Sounds/Fort_Weapons/Happy_Ghost/SoundWaves/Weapon_Happy_Ghost_Impact_Success_01}}
 {id=WavePlayer_8, properties {soundWaveAssetPtr=/Game/Sounds/Fort_Weapons/Happy_Ghost/SoundWaves/Bullet_Impact_Water_Med_01}}
 {id=WavePlayer_9, properties {soundWaveAssetPtr=/Game/Sounds/Fort_Weapons/Happy_Ghost/SoundWaves/Bullet_Impact_Water_Med_02}}
]
properties {
 attenuationSettings=&Fort_GadgetActivate_Att
 maxDistance=4000
 soundClassObject=&Fort_Impacts
 volumeMultiplier=0.7
 attenuationOverrides {bEnableReverbSend=false, falloffDistance=3000, nonSpatializedRadiusStart=1000, stereoSpread=2000, attenuationShapeExtents {x=2500}}
 concurrencySet [&RifleBulletImpact_Conc]
}
