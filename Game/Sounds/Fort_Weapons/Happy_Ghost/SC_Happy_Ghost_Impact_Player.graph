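# Sound cue for the Happy Ghost weapon's impact on a player.
# A mixer layers a random grappling hook impact, a quieter success impact wave and a full-volume random flesh punch crunch from four variants.
# A root modulator adds a slight upward pitch variation.
# Playback uses Happy Ghost impact attenuation with overrides, a third-person ranged-weapon sound class and rifle bullet impact concurrency.
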SoundCue /Game/Sounds/Fort_Weapons/Happy_Ghost/SC_Happy_Ghost_Impact_Player
root=Modulator_1
nodes [
 {id=Mixer_1, children [Random_0,WavePlayer_4,Random_1], properties {inputVolume [0.5,0.15,1]}}
 {id=Modulator_1, children [Mixer_1], properties {pitchMax=1.1, pitchMin=1}}
 {id=Random_0, children [WavePlayer_1,WavePlayer_2], properties {weights [1,1]}}
 {id=Random_1, children [WavePlayer_3,WavePlayer_5,WavePlayer_6,WavePlayer_7], properties {weights [1,1,1,1]}}
 {id=WavePlayer_1, properties {soundWaveAssetPtr=/Game/Sounds/Fort_Weapons/GrapplingHook/Grappling_Impact_01}}
 {id=WavePlayer_2, properties {soundWaveAssetPtr=/Game/Sounds/Fort_Weapons/GrapplingHook/Grappling_Impact_04}}
 {id=WavePlayer_3, properties {soundWaveAssetPtr=/Game/Sounds/Fort_Impact_Sounds/Flesh/punch_crunch_01}}
 {id=WavePlayer_4, properties {soundWaveAssetPtr=/Game/Sounds/Fort_Weapons/Happy_Ghost/SoundWaves/Weapon_Happy_Ghost_Impact_Success_01}}
 {id=WavePlayer_5, properties {soundWaveAssetPtr=/Game/Sounds/Fort_Impact_Sounds/Flesh/punch_crunch_02}}
 {id=WavePlayer_6, properties {soundWaveAssetPtr=/Game/Sounds/Fort_Impact_Sounds/Flesh/punch_crunch_03}}
 {id=WavePlayer_7, properties {soundWaveAssetPtr=/Game/Sounds/Fort_Impact_Sounds/Flesh/punch_crunch_04}}
]
properties {
 attenuationSettings=/Game/Sounds/Attenuation/Fort_Happy_Ghost_Impact_Att
 maxDistance=7000
 soundClassObject=&Fort_WeaponRanged_Fire_3P_PSM
 volumeMultiplier=1
 attenuationOverrides {bEnableReverbSend=false, falloffDistance=3000, nonSpatializedRadiusStart=1000, stereoSpread=2000, attenuationShapeExtents {x=2500}}
 concurrencySet [&RifleBulletImpact_Conc]
}
