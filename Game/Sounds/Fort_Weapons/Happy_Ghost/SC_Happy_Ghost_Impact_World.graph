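# Sound cue for the Happy Ghost weapon's impact on world surfaces, in the impacts sound class with gadget-activate attenuation and rifle bullet impact concurrency.
# A success impact wave plays immediately, mixed with a slightly delayed layer.
# That delayed layer blends a random pick of two borrowed Grappling Hook impacts with a much quieter Happy Ghost world impact.
# A root modulator adds slight upward pitch variation to the whole mix.
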SoundCue /Game/Sounds/Fort_Weapons/Happy_Ghost/SC_Happy_Ghost_Impact_World
root=Modulator_1
nodes [
 {id=Delay_1, children [Mixer_0], properties {delayMax=0.075, delayMin=0.05}}
 {id=Mixer_0, children [Random_0,WavePlayer_4], properties {inputVolume [1,0.15]}}
 {id=Mixer_1, children [Delay_1,WavePlayer_5], properties {inputVolume [1,0.75]}}
 {id=Modulator_1, children [Mixer_1], properties {pitchMax=1.1, pitchMin=1}}
 {id=Random_0, children [WavePlayer_1,WavePlayer_2], properties {weights [1,1]}}
 {id=WavePlayer_1, properties {soundWaveAssetPtr=/Game/Sounds/Fort_Weapons/GrapplingHook/Grappling_Impact_02}}
 {id=WavePlayer_2, properties {soundWaveAssetPtr=/Game/Sounds/Fort_Weapons/GrapplingHook/Grappling_Impact_03}}
 {id=WavePlayer_4, properties {soundWaveAssetPtr=/Game/Sounds/Fort_Weapons/Happy_Ghost/SoundWaves/Weapon_Happy_Ghost_Impact_World_01}}
 {id=WavePlayer_5, properties {soundWaveAssetPtr=/Game/Sounds/Fort_Weapons/Happy_Ghost/SoundWaves/Weapon_Happy_Ghost_Impact_Success_01}}
]
properties {
 attenuationSettings=&Fort_GadgetActivate_Att
 maxDistance=4000
 soundClassObject=&Fort_Impacts
 volumeMultiplier=1
 attenuationOverrides {bEnableReverbSend=false, falloffDistance=3000, nonSpatializedRadiusStart=1000, stereoSpread=2000, attenuationShapeExtents {x=2500}}
 concurrencySet [&RifleBulletImpact_Conc]
}
