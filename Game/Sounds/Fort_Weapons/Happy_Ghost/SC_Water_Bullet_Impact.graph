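# Sound cue for bullet impacts on water in the Happy_Ghost weapon sound set, using impact attenuation and the impacts sound class.
# A quality-level node gives its first slot ten random wave variations, while the other three slots share a three-wave random subset.
# The result passes through a fixed short delay and slight pitch randomization.
# Overridden concurrency caps simultaneous voices and stops the oldest, superseding the assigned rifle impact concurrency set.
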SoundCue /Game/Sounds/Fort_Weapons/Happy_Ghost/SC_Water_Bullet_Impact
root=Modulator_0
nodes [
 {id=Delay_8, children [QualityLevel_1], properties {delayMax=0.1, delayMin=0.1}}
 {id=Modulator_0, children [Delay_8], properties {pitchMax=1.02, pitchMin=0.98}}
 {id=QualityLevel_1, children [Random_1,Random_3,Random_3,Random_3]}
 {
  id=Random_1
  children [
   WavePlayer_8
   WavePlayer_9
   WavePlayer_10
   WavePlayer_11
   WavePlayer_12
   WavePlayer_13
   WavePlayer_14
   WavePlayer_15
   WavePlayer_16
   WavePlayer_17
  ]
  properties {weights [1,1,1,1,1,1,1,1,1,1]}
 }
 {id=Random_3, children [WavePlayer_18, WavePlayer_22, WavePlayer_23], properties {bShouldExcludeFromBranchCulling=true, weights [1,1,1]}}
 {id=WavePlayer_10, properties {soundWaveAssetPtr=/Game/Sounds/Fort_Weapons/Happy_Ghost/SoundWaves/Bullet_Impact_Water_Med_03}}
 {id=WavePlayer_11, properties {soundWaveAssetPtr=/Game/Sounds/Fort_Weapons/Happy_Ghost/SoundWaves/Bullet_Impact_Water_Med_04}}
 {id=WavePlayer_12, properties {soundWaveAssetPtr=/Game/Sounds/Fort_Weapons/Happy_Ghost/SoundWaves/Bullet_Impact_Water_Med_05}}
 {id=WavePlayer_13, properties {soundWaveAssetPtr=/Game/Sounds/Fort_Weapons/Happy_Ghost/SoundWaves/Bullet_Impact_Water_Med_06}}
 {id=WavePlayer_14, properties {soundWaveAssetPtr=/Game/Sounds/Fort_Weapons/Happy_Ghost/SoundWaves/Bullet_Impact_Water_Med_07}}
 {id=WavePlayer_15, properties {soundWaveAssetPtr=/Game/Sounds/Fort_Weapons/Happy_Ghost/SoundWaves/Bullet_Impact_Water_Med_08}}
 {id=WavePlayer_16, properties {soundWaveAssetPtr=/Game/Sounds/Fort_Weapons/Happy_Ghost/SoundWaves/Bullet_Impact_Water_Med_09}}
 {id=WavePlayer_17, properties {soundWaveAssetPtr=/Game/Sounds/Fort_Weapons/Happy_Ghost/SoundWaves/Bullet_Impact_Water_Med_10}}
 {id=WavePlayer_18, properties {soundWaveAssetPtr=/Game/Sounds/Fort_Weapons/Happy_Ghost/SoundWaves/Bullet_Impact_Water_Med_01}}
 {id=WavePlayer_22, properties {soundWaveAssetPtr=&Bullet_Impact_Water_Med_03}}
 {id=WavePlayer_23, properties {soundWaveAssetPtr=&Bullet_Impact_Water_Med_06}}
 {id=WavePlayer_8, properties {soundWaveAssetPtr=&Bullet_Impact_Water_Med_01}}
 {id=WavePlayer_9, properties {soundWaveAssetPtr=/Game/Sounds/Fort_Weapons/Happy_Ghost/SoundWaves/Bullet_Impact_Water_Med_02}}
]
properties {
 attenuationSettings=/Game/Sounds/Attenuation/Fort_Bullet_Impact_Att
 bOverrideConcurrency=true
 maxDistance=5300
 soundClassObject=&Fort_Impacts
 volumeMultiplier=1.77
 concurrencyOverrides {maxCount=5, resolutionRule=StopOldest, volumeScale=0.5}
 concurrencySet [&RifleBulletImpact_Conc]
}
